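; Legacy prototype puzzle fountain with a wish interaction that tosses out loot.
; Interacting disables further use, waits briefly, then spawns loot-tier pickups at the billboard and tosses each along its forward vector with random spread.
; TurnOnFountainBase sets PuzzleSolved on the authority, while remotes play the activation Niagara effect and sound.
; OnRep_PuzzleSolved reveals the base fountain mesh once solved.
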(event Custom|TurnOnFountainBase
  (Networking|SwitchHasAuthority
    (:Authority
      (Variables|Default|SetPuzzleSolved true))
    (:Remote
      (Components|Activation|Activate (Variables|Default|GetNS_FountainActivate) false)
      (Audio|PlaySoundatLocation "/Game/Sounds/Fort_Gadgets/Airstrike/airstrike_explosion_Cue.airstrike_explosion_Cue" (Transformation|GetWorldLocation (Variables|Default|GetNS_FountainActivate))))))

(event Interaction|EventBlueprintOnInteract
  (params (InteractingPawn :type Fort%20Pawn%20Object%20Reference) (InteractionBeingAttempted :type EInteractionBeingAttempted%20Enum :default "FirstInteraction"))
  (bind _billboard (Variables|Default|GetBillboard))
  (bind _returnvalue (Transformation|GetForwardVector _billboard))
  (bind _tossvelocity (Variables|Default|GetTossVelocity))
  (bind _returnvalue_2 (Math|Conversions|ToVector(Float) _tossvelocity))
  (bind _returnvalue_3 (* _returnvalue _returnvalue_2))
  (Variables|Default|SetCanInteract false)
  (Utilities|FlowControl|Delay 1.5)
  (Variables|Default|SetImpulse _returnvalue_3)
  (bind _returnvalue_4 (Inventory|SpawnPickupInWorldWithLootTier (Variables|Default|GetLootTierToFire) (Transformation|GetWorldLocation _billboard) 0 0 true true _returnvalue_3))
  (Variables|Default|SetLoot _returnvalue_4)
  (for _array_element (Variables|Default|GetLoot)
    (bind _asfort_pickup_athena (Utilities|Casting|CastToFortPickupAthena _array_element))
    (Pickup|TossPickupinDirection _asfort_pickup_athena (+ (Math|Vector|MakeVector (* (Variables|Default|GetTossSpreadMultiplier) (Math|Random|RandomFloatinRange -4e+02 4e+02)) (* (Math|Random|RandomFloatinRange -4e+02 4e+02) (Variables|Default|GetTossSpreadMultiplier)) (* (Math|Random|RandomFloatinRange 1e+02 2.5e+02) (Variables|Default|GetTossSpreadMultiplier))) (* (Transformation|GetForwardVector (Variables|Default|GetBillboard)) (Variables|Default|GetTossVelocity))) self false false)))

(fn BlueprintCanInteract
  (params (InteractingPawn :type Fort%20Pawn%20Object%20Reference) (InteractionBeingAttempted :type EInteractionBeingAttempted%20Enum :default "FirstInteraction") (InteractionType :type TInteractionType%20Enum :default "IT_NoInteraction"))
  (return (Variables|Default|GetCanInteract)))

(fn BlueprintGetInteractionString
  (params (InteractingPawn :type Fort%20Pawn%20Object%20Reference) (InteractionBeingAttempted :type EInteractionBeingAttempted%20Enum :default "FirstInteraction"))
  (return (Variables|Default|GetWishText)))

(fn OnRep_PuzzleSolved
  (if (Variables|Default|GetPuzzleSolved)
    (Rendering|SetVisibility (Variables|Default|GetBaseFountainMesh) true false)))

(fn ConstructionScript
  (|CallParentFunction))
