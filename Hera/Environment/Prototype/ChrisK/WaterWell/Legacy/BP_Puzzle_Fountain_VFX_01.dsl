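; Legacy Blueprint for the VFX of a water-well fountain puzzle, showing the fountain's water once the puzzle is solved.
; The TurnOnFountainTop event sets the replicated PuzzleSolved flag only on the server authority.
; Its OnRep handler makes the water spline and center water meshes visible, keeping the visual state in sync across clients.
; The construction script only calls its parent.
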(event Custom|TurnOnFountainTop
  (Networking|SwitchHasAuthority
    (:Authority
      (Variables|Default|SetPuzzleSolved true))))

(fn OnRep_PuzzleSolved
  (Rendering|SetVisibility (Variables|Default|GetSM_Fountain_WaterSplines_01) true false)
  (Rendering|SetVisibility (Variables|Default|GetSM_Outdoor_FountainCenterWater_A) true false))

(fn ConstructionScript
  (|CallParentFunction))
